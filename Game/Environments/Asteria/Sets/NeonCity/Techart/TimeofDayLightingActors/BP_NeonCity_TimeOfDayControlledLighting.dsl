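; Neon City light actor that switches its light components and mesh emissive between day and night, when time-of-day toggling is enabled.
; Construction and non-authority BeginPlay gather and tag lights; BeginPlay also registers for day-phase changes.
; Phase changes wait a random 0–2 seconds, then UpdateLights turns lights off for Day, on otherwise.
; A separate Update event refreshes all same-class actors from the global time-of-day material parameter, treating hours 8–20 as day.
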(event EventBeginPlay
  (Networking|SwitchHasAuthority
    (:Remote
      (Utilities|FlowControl|Sequence
        (:then_0
          (bind _output_get (Variables|Default|SetLights (Actor|GetComponentsByClass "/Script/Engine.LightComponent")))
          (for _array_element _output_get
            (Variables|/Script/Engine.ActorComponent|SetComponentTags (Variables|Default|GetComponent%20Tags) _array_element)))
        (:then_1
          (Class|FortAsyncActionTimeOfDayManagerReady|GetTimeOfDayManagerReady)
          (TimeOfDay|RegisterforDayPhaseChangedEvent (TimeOfDay|GetGlobalTimeOfDayManager) (delegate Custom|Time%20of%20Day%20Phase%20Change) false)
          (CallFunction|UpdateLights (Utilities|Enum|Equal(Enum) (stale Custom|Time%20of%20Day%20Phase%20Change CurrentDayPhase) "Day")))))))

(event Collision|EventActorBeginOverlap
  (params (OtherActor :type Actor%20Object%20Reference)))

(event Custom|Time%20of%20Day%20Phase%20Change
  (params (CurrentDayPhase :type EFortDayPhase%20Enum :default "Morning") (PreviousDayPhase :type EFortDayPhase%20Enum :default "Morning") (bAtCreation :type Boolean))
  (Utilities|FlowControl|Delay (Math|Random|RandomFloatinRange 0.0 2.0))
  (CallFunction|UpdateLights (Utilities|Enum|Equal(Enum) CurrentDayPhase "Day")))

(event Custom|Update
  (bind _returnvalue (Rendering|Material|GetScalarParameterValue "/Game/Packages/Fortress_SharedMaterials/GlobalMaterialParameters/FortniteMaterialParameters.FortniteMaterialParameters" "TimeOfDay_0-1" :!class /Script/Engine.KismetMaterialLibrary))
  (bind _outactors (Actor|GetAllActorsOfClass (Utilities|GetClass self)))
  (for _array_element _outactors
    (CallFunction|UpdateLights _array_element (and (> 2e+01 (* _returnvalue 24.0)) (> (* _returnvalue 24.0) 8.0)))))

(fn UpdateLights
  (params (Condition :type Boolean))
  (bind _lights (Variables|Default|GetLights))
  (bind _staticmeshcomponent (Variables|Debug|GetStaticMeshComponent))
  (if (Variables|Default|GetToggle%20By%20Time%20of%20Day)
    (if Condition
      (for _array_element _lights
        (Rendering|Components|Light|SetIntensity _array_element 0.0)
        (Rendering|SetVisibility _array_element false true))
      (Rendering|Material|SetVectorParameterValueonMaterials _staticmeshcomponent "EmissiveMultiply" (Variables|Default|GetMaterialEmissive_DayMult))
      (Variables|Default|SetLightisCurrentOn false)
      (else
        (for _array_element_2 _lights
          (Rendering|Components|Light|SetIntensity _array_element_2 (Variables|Default|GetLight%20Intensity))
          (Rendering|SetVisibility _array_element_2 true true))
        (Rendering|Material|SetVectorParameterValueonMaterials _staticmeshcomponent "EmissiveMultiply" (Variables|Default|GetMaterialEmissive_NightMult))
        (Variables|Default|SetLightisCurrentOn true)))))

(fn ConstructionScript
  (bind _output_get (Variables|Default|SetLights (Actor|GetComponentsByClass "/Script/Engine.LightComponent")))
  (for _array_element _output_get
    (Variables|/Script/Engine.ActorComponent|SetComponentTags (Variables|Default|GetComponent%20Tags) _array_element)))
